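# Cosmetic AnimBP for the Ephemeral female armor body, copying the parent mesh pose into a Control Rig driven by neck bend.
# Bone constraints then position attachments and armor pieces, while additive bone transforms adjust spikes, leaves and drapes.
# The EventGraph derives their alphas from ref-pose bone deltas, though left drape's input is constant zero.
# Gears turn on crouch changes and spin during crouch movement; melee weapons offset the backpack.
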AnimBlueprint /BRCosmetics/Characters/Player/Female/Medium/Bodies/F_MED_Ephemeral_Armor/Meshes/F_MED_Ephemeral_Armor_AnimBP
target_skeleton=&/F_MED_Ephemeral_Armor_Skeleton
graphs [
 AnimGraph {nodes [
   n1 = "Constraint - dyn_L_frontChest_oval_scale (PPP)"!(ComponentPose, Alpha=1, ConstraintWeights_0=0.8, ConstraintWeights_1:float=0.1, ConstraintWeights_2:float=0.1) -> (Pose) {
    alphaBoolBlend {blendOption=Linear}
    boneToModify {boneName=dyn_L_frontChest_oval_scale}
    constraintSetup [{targetBone {boneName=spine_05}}, {targetBone {boneName=neck_01}}, {targetBone {boneName=dyn_L_upperarm_sleeveFollow}}]
    constraintWeights [1,1,1]
   }
   "Constraint - dyn_R_frontChest_oval_scale (PPP)"!(ComponentPose=@n1, Alpha=1, ConstraintWeights_0=0.8, ConstraintWeights_1:float=0.1, ConstraintWeights_2:float=0.1) -> (Pose) {
    alphaBoolBlend {blendOption=Linear}
    boneToModify {boneName=dyn_R_frontChest_oval_scale}
    constraintSetup [{targetBone {boneName=spine_05}}, {targetBone {boneName=neck_01}}, {targetBone {boneName=dyn_R_upperarm_sleeveFollow}}]
    constraintWeights [1,1,1]
   }
   n2 = Control Rig!(Source=Copy Pose From Mesh() {bUseAttachedParent=true}, Alpha, Neck Back:double=Get NeckBack() -> (NeckBack:double)) -> (Pose) {
    alphaBoolBlend {blendOption=Linear}
    controlRigAssetReference {blueprintRigClass=&/F_MED_Ephemeral_Armor_CtrlRig.F_MED_Ephemeral_Armor_CtrlRig_C}
   }
   n3 = "Constraint - attach_backpack (PPPP)"!(
    ComponentPose=Local To Component(LocalPose=@n2)
    Alpha=1
    ConstraintWeights_0=0.2198
    ConstraintWeights_1:float=0.6072
    ConstraintWeights_2:float=0.0332
    ConstraintWeights_3:float=0.1399
   ) -> (Pose) {
    alphaBoolBlend {blendOption=Linear}
    boneToModify {boneName=attach_backpack}
    constraintSetup [{targetBone {boneName=spine_05}}, {targetBone {boneName=spine_04}}, {targetBone {boneName=spine_03}}, {targetBone {boneName=spine_02}}]
    constraintWeights [1,1,1,1]
   }
   n4 = "Constraint - attach_cape (P)"(ComponentPose=@n3, Alpha=1, ConstraintWeights_0=1) {
    alphaBoolBlend {blendOption=Linear}
    boneToModify {boneName=attach_cape}
    constraintSetup [{targetBone {boneName=spine_05}}]
    constraintWeights [1]
   }
   n5 = "Constraint - dyn_L_upperarm_sleeveFollow (R)"(ComponentPose=@n4, Alpha=1, ConstraintWeights_0=1) {
    alphaBoolBlend {blendOption=Linear}
    alphaScaleBiasClamp {scale=0.6}
    boneToModify {boneName=dyn_L_upperarm_sleeveFollow}
    constraintSetup [{transformType=Rotation, targetBone {boneName=upperarm_l}}]
    constraintWeights [1]
   }
   n6 = "Constraint - dyn_R_upperarm_sleeveFollow (R)"(ComponentPose=@n5, Alpha=1, ConstraintWeights_0=1) {
    alphaBoolBlend {blendOption=Linear}
    alphaScaleBiasClamp {scale=0.6}
    boneToModify {boneName=dyn_R_upperarm_sleeveFollow}
    constraintSetup [{transformType=Rotation, targetBone {boneName=upperarm_r}}]
    constraintWeights [1]
   }
   n7 = "Transform (Modify) Bone - Bone: dyn_L_calfSpike_scale"(ComponentPose=@n6, Alpha=Get L calf bend() -> (L calf bend:double), Translation="0,0,0", Rotation="0,0,0", Scale="1,0.5,1") {
    scaleMode=BMM_Additive
    scaleSpace=BCS_BoneSpace
    alphaBoolBlend {blendOption=Linear}
    boneToModify {boneName=dyn_L_calfSpike_scale}
    scale {y=0.5}
   }
   n8 = "Transform (Modify) Bone - Bone: dyn_R_calfSpike_scale"(ComponentPose=@n7, Alpha=Get R calf bend() -> (R calf bend:double), Translation="0,0,0", Rotation="0,0,0", Scale="1,0.5,1") {
    scaleMode=BMM_Additive
    scaleSpace=BCS_BoneSpace
    alphaBoolBlend {blendOption=Linear}
    boneToModify {boneName=dyn_R_calfSpike_scale}
    scale {y=0.5}
   }
   n9 = "Transform (Modify) Bone - Bone: dyn_L_thighSpike_scale"(ComponentPose=@n8, Alpha=Get L thigh twist() -> (L thigh twist:double), Translation="0,0,0", Rotation="0,0,0", Scale="1,0.7,1") {
    scaleMode=BMM_Additive
    scaleSpace=BCS_BoneSpace
    alphaBoolBlend {blendOption=Linear}
    boneToModify {boneName=dyn_L_thighSpike_scale}
    scale {y=0.7}
   }
   n10 = "Transform (Modify) Bone - Bone: dyn_R_thighSpike_scale"(ComponentPose=@n9, Alpha=Get R thigh twist() -> (R thigh twist:double), Translation="0,0,0", Rotation="0,0,0", Scale="1,0.7,1") {
    scaleMode=BMM_Additive
    scaleSpace=BCS_BoneSpace
    alphaBoolBlend {blendOption=Linear}
    boneToModify {boneName=dyn_R_thighSpike_scale}
    scale {y=0.7}
   }
   n11 = "Constraint - dyn_backDiamond_constrain_upper (PPPPPP)"!(
    ComponentPose=@n10
    Alpha=1
    ConstraintWeights_0=0.1284
    ConstraintWeights_1:float=0.0278
    ConstraintWeights_2:float=0.4819
    ConstraintWeights_3:float=0.3466
    ConstraintWeights_4:float=0.0076
    ConstraintWeights_5:float=0.0076
   ) -> (Pose) {
    alphaBoolBlend {blendOption=Linear}
    boneToModify {boneName=dyn_backDiamond_constrain_upper}
    constraintSetup [
     {targetBone {boneName=spine_02}}
     {targetBone {boneName=spine_03}}
     {targetBone {boneName=spine_04}}
     {targetBone {boneName=spine_05}}
     {targetBone {boneName=upperarm_l}}
     {targetBone {boneName=upperarm_r}}
    ]
    constraintWeights [1,1,1,1,1,1]
   }
   n12 = "Constraint - dyn_backDiamond_constrain_mid (PPPP)"!(
    ComponentPose=@n11
    Alpha=1
    ConstraintWeights_0=0.2353
    ConstraintWeights_1:float=0.0902
    ConstraintWeights_2:float=0.5368
    ConstraintWeights_3:float=0.1376
   ) -> (Pose) {
    alphaBoolBlend {blendOption=Linear}
    boneToModify {boneName=dyn_backDiamond_constrain_mid}
    constraintSetup [{targetBone {boneName=spine_02}}, {targetBone {boneName=spine_03}}, {targetBone {boneName=spine_04}}, {targetBone {boneName=spine_05}}]
    constraintWeights [1,1,1,1]
   }
   n13 = "Constraint - dyn_backDiamond_constrain_lower (PPPP)"!(
    ComponentPose=@n12
    Alpha=1
    ConstraintWeights_0=0.3134
    ConstraintWeights_1:float=0.1442
    ConstraintWeights_2:float=0.4729
    ConstraintWeights_3:float=0.0694
   ) -> (Pose) {
    alphaBoolBlend {blendOption=Linear}
    boneToModify {boneName=dyn_backDiamond_constrain_lower}
    constraintSetup [{targetBone {boneName=spine_02}}, {targetBone {boneName=spine_03}}, {targetBone {boneName=spine_04}}, {targetBone {boneName=spine_05}}]
    constraintWeights [1,1,1,1]
   }
   n14 = "Constraint - dyn_L_upperarmElement_constrain (P)"!(ComponentPose=@n13, Alpha=1, ConstraintWeights_0=1, ConstraintWeights_1:float=0) -> (Pose) {
    alphaBoolBlend {blendOption=Linear}
    boneToModify {boneName=dyn_L_upperarmElement_constrain}
    constraintSetup [{targetBone {boneName=upperarm_twist_01_l}}]
    constraintWeights [1,0]
   }
   n15 = "Constraint - dyn_L_lowerarmFin_constrain (PPP)"!(ComponentPose=@n14, Alpha=1, ConstraintWeights_0=0.6189, ConstraintWeights_1:float=0.3293, ConstraintWeights_2:float=0.0518) -> (Pose) {
    alphaBoolBlend {blendOption=Linear}
    boneToModify {boneName=dyn_L_lowerarmFin_constrain}
    constraintSetup [{targetBone {boneName=lowerarm_l}}, {targetBone {boneName=lowerarm_twist_02_l}}, {targetBone {boneName=lowerarm_twist_01_l}}]
    constraintWeights [1,1,1]
   }
   n16 = "Constraint - dyn_R_upperarmElement_constrain (P)"!(ComponentPose=@n15, Alpha=1, ConstraintWeights_0=1, ConstraintWeights_1:float=0) -> (Pose) {
    alphaBoolBlend {blendOption=Linear}
    boneToModify {boneName=dyn_R_upperarmElement_constrain}
    constraintSetup [{targetBone {boneName=upperarm_twist_01_r}}]
    constraintWeights [1,0]
   }
   n17 = "Constraint - dyn_R_lowerarmFin_constrain (PPP)"!(ComponentPose=@n16, Alpha=1, ConstraintWeights_0=0.6189, ConstraintWeights_1:float=0.3293, ConstraintWeights_2:float=0.0518) -> (Pose) {
    alphaBoolBlend {blendOption=Linear}
    boneToModify {boneName=dyn_R_lowerarmFin_constrain}
    constraintSetup [{targetBone {boneName=lowerarm_r}}, {targetBone {boneName=lowerarm_twist_02_r}}, {targetBone {boneName=lowerarm_twist_01_r}}]
    constraintWeights [1,1,1]
   }
   n18 = "Constraint - dyn_L_kneecap (PP)"!(ComponentPose=@n17, Alpha=1, ConstraintWeights_0=0.3, ConstraintWeights_1:float=0.7) -> (Pose) {
    alphaBoolBlend {blendOption=Linear}
    boneToModify {boneName=dyn_L_kneecap}
    constraintSetup [{targetBone {boneName=thigh_l}}, {targetBone {boneName=calf_l}}]
    constraintWeights [1,1]
   }
   n19 = "Constraint - dyn_R_kneecap (PP)"!(ComponentPose=@n18, Alpha=1, ConstraintWeights_0=0.3, ConstraintWeights_1:float=0.7) -> (Pose) {
    alphaBoolBlend {blendOption=Linear}
    boneToModify {boneName=dyn_R_kneecap}
    constraintSetup [{targetBone {boneName=thigh_r}}, {targetBone {boneName=calf_r}}]
    constraintWeights [1,1]
   }
   n20 = Get gearRotation() -> (gearRotation:Rotator)
   n21 = "Transform (Modify) Bone - Bone: dyn_R_innerGear"(ComponentPose=@n19, Alpha=1, Translation="0,0,0", Rotation=@n20, Scale="1,1,1") {rotationMode=BMM_Additive, rotationSpace=BCS_BoneSpace, alphaBoolBlend {blendOption=Linear}, boneToModify {boneName=dyn_R_innerGear}}
   n22 = "Transform (Modify) Bone - Bone: dyn_R_outerGear"(ComponentPose=@n21, Alpha=1, Translation="0,0,0", Rotation=@n20, Scale="1,1,1") {rotationMode=BMM_Additive, rotationSpace=BCS_BoneSpace, alphaBoolBlend {blendOption=Linear}, boneToModify {boneName=dyn_R_outerGear}}
   n23 = "Transform (Modify) Bone - Bone: dyn_L_innerGear"(ComponentPose=@n22, Alpha=1, Translation="0,0,0", Rotation=@n20, Scale="1,1,1") {rotationMode=BMM_Additive, rotationSpace=BCS_BoneSpace, alphaBoolBlend {blendOption=Linear}, boneToModify {boneName=dyn_L_innerGear}}
   n24 = "Transform (Modify) Bone - Bone: dyn_L_outerGear"(ComponentPose=@n23, Alpha=1, Translation="0,0,0", Rotation=@n20, Scale="1,1,1") {rotationMode=BMM_Additive, rotationSpace=BCS_BoneSpace, alphaBoolBlend {blendOption=Linear}, boneToModify {boneName=dyn_L_outerGear}}
   n25 = "Transform (Modify) Bone - Bone: dyn_L_LEAF_base"(ComponentPose=@n24, Alpha=Get L_clavicle_follow() -> (L_clavicle_follow:double), Translation="0,-0.5,0", Rotation="0,0,20", Scale="1,1,1") {
    rotationMode=BMM_Additive
    rotationSpace=BCS_BoneSpace
    translationMode=BMM_Additive
    translationSpace=BCS_BoneSpace
    alphaBoolBlend {blendOption=Linear}
    alphaScaleBiasClamp {bInterpResult=true}
    boneToModify {boneName=dyn_L_LEAF_base}
    rotation {roll=20}
    translation {y=-0.5}
   }
   n26 = "Transform (Modify) Bone - Bone: dyn_R_LEAF_base"(ComponentPose=@n25, Alpha=Get R_clavicle_follow() -> (R_clavicle_follow:double), Translation="0,0.5,0", Rotation="0,0,20", Scale="1,1,1") {
    rotationMode=BMM_Additive
    rotationSpace=BCS_BoneSpace
    translationMode=BMM_Additive
    translationSpace=BCS_BoneSpace
    alphaBoolBlend {blendOption=Linear}
    alphaScaleBiasClamp {bInterpResult=true}
    boneToModify {boneName=dyn_R_LEAF_base}
    rotation {roll=20}
    translation {y=0.5}
   }
   n27 = Get Rotation Out() -> (Rotation Out:Rotator)
   n28 = "Transform (Modify) Bone - Bone: dyn_R_innerGear"(ComponentPose=@n26, Alpha=1, Translation="0,0,0", Rotation=@n27, Scale="1,1,1") {rotationMode=BMM_Additive, rotationSpace=BCS_BoneSpace, alphaBoolBlend {blendOption=Linear}, boneToModify {boneName=dyn_R_innerGear}}
   n29 = "Transform (Modify) Bone - Bone: dyn_R_outerGear"(ComponentPose=@n28, Alpha=1, Translation="0,0,0", Rotation=@n27, Scale="1,1,1") {rotationMode=BMM_Additive, rotationSpace=BCS_BoneSpace, alphaBoolBlend {blendOption=Linear}, boneToModify {boneName=dyn_R_outerGear}}
   n30 = "Transform (Modify) Bone - Bone: dyn_L_innerGear"(ComponentPose=@n29, Alpha=1, Translation="0,0,0", Rotation=@n27, Scale="1,1,1") {rotationMode=BMM_Additive, rotationSpace=BCS_BoneSpace, alphaBoolBlend {blendOption=Linear}, boneToModify {boneName=dyn_L_innerGear}}
   n31 = "Transform (Modify) Bone - Bone: dyn_L_outerGear"(ComponentPose=@n30, Alpha=1, Translation="0,0,0", Rotation=@n27, Scale="1,1,1") {rotationMode=BMM_Additive, rotationSpace=BCS_BoneSpace, alphaBoolBlend {blendOption=Linear}, boneToModify {boneName=dyn_L_outerGear}}
   n32 = "Transform (Modify) Bone - Bone: dyn_L_drape"(ComponentPose=@n31, Alpha=Get spineRotate left() -> (spineRotate left:double), Translation="0,-3,0", Rotation="0,0,0", Scale="1,1,1") {
    translationMode=BMM_Additive
    translationSpace=BCS_BoneSpace
    alphaBoolBlend {blendOption=Linear}
    alphaScaleBiasClamp {bInterpResult=true, interpSpeedDecreasing=20, interpSpeedIncreasing=20}
    boneToModify {boneName=dyn_L_drape}
    translation {y=-3}
   }
   n33 = "Transform (Modify) Bone - Bone: dyn_R_drape"(ComponentPose=@n32, Alpha=Get spineRotate right() -> (spineRotate right:double), Translation="0,3,0", Rotation="0,0,0", Scale="1,1,1") {
    translationMode=BMM_Additive
    translationSpace=BCS_BoneSpace
    alphaBoolBlend {blendOption=Linear}
    alphaScaleBiasClamp {bInterpResult=true, interpSpeedDecreasing=20, interpSpeedIncreasing=20}
    boneToModify {boneName=dyn_R_drape}
    translation {y=3}
   }
   n34 = "Transform (Modify) Bone - Bone: dyn_L_scaleProblemArea"(ComponentPose=@n33, Alpha=1, Translation="-1,0,0", Rotation="0,0,0", Scale="0.4,0.4,0.4") {
    scaleMode=BMM_Additive
    scaleSpace=BCS_ParentBoneSpace
    translationMode=BMM_Additive
    translationSpace=BCS_BoneSpace
    alphaBoolBlend {blendOption=Linear}
    alphaScaleBiasClamp {interpSpeedDecreasing=20, interpSpeedIncreasing=20}
    boneToModify {boneName=dyn_L_scaleProblemArea}
    scale=<0.4 0.4 0.4>
    translation {x=-1}
   }
   n35 = "Transform (Modify) Bone - Bone: dyn_R_scaleProblemArea"(ComponentPose=@n34, Alpha=1, Translation="1,0,0", Rotation="0,0,0", Scale="0.4,0.4,0.4") {
    scaleMode=BMM_Additive
    scaleSpace=BCS_ParentBoneSpace
    translationMode=BMM_Additive
    translationSpace=BCS_BoneSpace
    alphaBoolBlend {blendOption=Linear}
    alphaScaleBiasClamp {interpSpeedDecreasing=20, interpSpeedIncreasing=20}
    boneToModify {boneName=dyn_R_scaleProblemArea}
    scale=<0.4 0.4 0.4>
    translation {x=1}
   }
   n36 = "Transform (Modify) Bone - Bone: attach_backpack"!(
    ComponentPose=@n35
    bAlphaBoolEnabled:bool=Get PickaxeEquipped() -> (PickaxeEquipped:bool)
    Translation="0,-2,0"
    Rotation="0,10,0"
    Scale="1,1,1"
   ) -> (Pose) {
    alphaInputType=Bool
    rotationMode=BMM_Additive
    rotationSpace=BCS_BoneSpace
    translationMode=BMM_Additive
    translationSpace=BCS_BoneSpace
    alphaBoolBlend {blendOption=Linear}
    boneToModify {boneName=attach_backpack}
    rotation {yaw=10}
    translation {y=-2}
   }
   Output Pose(Result=Component To Local(ComponentPose=@n36)) {layerGroup=None}
  ]}
 EventGraph {nodes [
   n1 = Event BlueprintUpdateAnimation()
   > n2 = Sequence!(execute) -> (then_0, then_1, then_2:Exec)
   n3 = Sequence!(execute=@n2.then_1) -> (then_0, then_1)
   > n4 = Branch(Condition=Get bIsCrouching())
   Set crouchEventBool(execute=[@n31.then, @n4.then], crouchEventBool:bool=true) -> (Output_Get:bool)
   Set crouchEventBool(execute=[@n32.then, @n4.else], crouchEventBool=false)
   n5 = Branch(execute=@n3.then_1, Condition=Get crouchEventBool() -> (crouchEventBool:bool))
   > Set onCrouchRotateAlpha(onCrouchRotateAlpha:double=FInterpTo(Current=Get onCrouchRotateAlpha() -> (onCrouchRotateAlpha:double), Target=0, DeltaTime=@n1.DeltaTimeX, InterpSpeed=4)) -> (Output_Get:double)
   Set onCrouchRotateAlpha(execute=@n5.else, onCrouchRotateAlpha=FInterpTo(Current=Get onCrouchRotateAlpha(), Target=1, DeltaTime=@n1.DeltaTimeX, InterpSpeed=4))
   n6 = Sequence(execute=@n2.then_2)
   > Branch(Condition=Get IsCrouchMovingBool() -> (IsCrouchMovingBool:bool))
   > n7 = Create Constant Rotation Function!(execute, RotationRate:Rotator="0,3.5,0", RotationVar:"Rotator (by ref)"=Get Rotation Out(), DeltaTime:double=0.03) -> (then, RotationOut:Rotator)
   > Set Rotation Out(Rotation Out:Rotator=@n7) -> (Output_Get:Rotator)
   n8 = GetOwnerSkeletalMeshComponent()
   n9 = BreakTransform(InTransform=GetDeltaTransformFromRefPose(self=@n8, BoneName=head, BaseName=spine_05))
   Set NeckBack(execute=@n6.then_1, NeckBack:double=MapRangeClamped(Value=BreakRotator(InRot=@n9.Rotation).Yaw, InRangeA=0, InRangeB=-12)) -> (Output_Get:double)
   n10 = BreakRotator(InRot=BreakTransform(InTransform=GetDeltaTransformFromRefPose(self=@n8, BoneName=clavicle_l, BaseName=spine_05)).Rotation)
   n11 = "Absolute (Float)"(A=@n10.Yaw)
   n12 = "Absolute (Float)"(A=@n10.Pitch)
   > Set L_clavicle_follow(L_clavicle_follow:double=MapRangeClamped(Value="Max (Float)"(A=@n12, B=@n11), InRangeA=0, InRangeB=7)) -> (Output_Get:double)
   n13 = BreakRotator(InRot=BreakTransform(InTransform=GetDeltaTransformFromRefPose(self=@n8, BoneName=clavicle_r, BaseName=spine_05)).Rotation)
   n14 = "Absolute (Float)"(A=@n13.Yaw)
   n15 = "Absolute (Float)"(A=@n13.Pitch)
   > Set R_clavicle_follow(R_clavicle_follow:double=MapRangeClamped(Value="Max (Float)"(A=@n15, B=@n14), InRangeA=0, InRangeB=7)) -> (Output_Get:double)
   n16 = BreakTransform(InTransform=GetDeltaTransformFromRefPose(self=@n8, BoneName=calf_l, BaseName=thigh_l))
   > Set L calf bend(L calf bend:double=MapRangeClamped(Value=BreakRotator(InRot=@n16.Rotation).Yaw, InRangeA=70, InRangeB=100)) -> (Output_Get:double)
   n17 = BreakTransform(InTransform=GetDeltaTransformFromRefPose(self=@n8, BoneName=calf_r, BaseName=thigh_r))
   > Set R calf bend(R calf bend:double=MapRangeClamped(Value=BreakRotator(InRot=@n17.Rotation).Yaw, InRangeA=70, InRangeB=100)) -> (Output_Get:double)
   n18 = BreakRotator(InRot=BreakTransform(InTransform=GetDeltaTransformFromRefPose(self=@n8, BoneName=thigh_l, BaseName=pelvis)).Rotation)
   n19 = MapRangeClamped(Value=@n18.Yaw, InRangeA=0, InRangeB=-40)
   n20 = "Absolute (Float)"(A=MapRangeClamped(Value=@n18, InRangeA=10, InRangeB=20))
   > Set L thigh twist(L thigh twist:double="Clamp (Float)"(Value=float - float(A=@n20, B=@n19), Min=0, Max=1)) -> (Output_Get:double)
   n21 = BreakRotator(InRot=BreakTransform(InTransform=GetDeltaTransformFromRefPose(self=@n8, BoneName=thigh_r, BaseName=pelvis)).Rotation)
   n22 = "Absolute (Float)"(A=MapRangeClamped(Value=@n21, InRangeA=10, InRangeB=20))
   n23 = MapRangeClamped(Value=@n21.Yaw, InRangeA=0, InRangeB=-40)
   > Set R thigh twist(R thigh twist:double="Clamp (Float)"(Value=float - float(A=@n22, B=@n23), Min=0, Max=1)) -> (Output_Get:double)
   n24 = BreakTransform(InTransform=GetDeltaTransformFromRefPose(self=@n8, BoneName=spine_05, BaseName=spine_03))
   > Set spineRotate right(spineRotate right:double=MapRangeClamped(Value=BreakRotator(InRot=@n24.Rotation), InRangeA=5, InRangeB=40)) -> (Output_Get:double)
   > Set spineRotate left(spineRotate left:double=MapRangeClamped(Value=0, InRangeA=-5, InRangeB=-40)) -> (Output_Get:double)
   n25 = Get onCrouchRotateAlpha()
   Set gearRotation(execute=@n2.then_0, gearRotation:Rotator=MakeRotator(Yaw=float * float(A=@n25, B=300))) -> (Output_Get:Rotator)
   n26 = NAND Boolean(A=Get bIsAccelerating2D(), B=Get bIsCrouching()) {is=title K2Node_CommutativeAssociativeBinaryOperator}
   n27 = Get bIsRiding() -> (bIsRiding:bool)
   > n28 = Set IsCrouchMovingBool(IsCrouchMovingBool:bool=AND Boolean(A=NOT Boolean(A=@n26), B=NOT Boolean(A=@n27))) -> (Output_Get:bool)
   > Set isCrouchMovingFloat(isCrouchMovingFloat:double="To Float (Boolean)"(InBool=@n28)) -> (Output_Get:double)
   Event OnWeaponChanged()
   > Set PickaxeEquipped(PickaxeEquipped:bool=false) -> (Output_Get:bool)
   > n29 = Cast To FortPlayerAnimInstance(Object=GetAnimInstance(self=GetOwnerSkeletalMeshComponent()))
   > n30 = Switch on EFortWeaponCoreAnimation!(execute, Selection:EFortWeaponCoreAnimation Enum=Get WeaponCoreAnim(self=@n29)) -> (
    Melee:Exec
    Pistol:Exec
    Shotgun:Exec
    PaperBlueprint:Exec
    Rifle:Exec
    MeleeOneHand:Exec
    MachinePistol:Exec
    RocketLauncher:Exec
    GrenadeLauncher:Exec
    GoingCommando:Exec
    AssaultRifle:Exec
    TacticalShotgun:Exec
    SniperRifle:Exec
    TrapPlacement:Exec
    ShoulderLauncher:Exec
    AbilityDecoTool:Exec
    Crossbow:Exec
    C4:Exec
    RemoteControl:Exec
    DualWield:Exec
    AR_BullPup:Exec
    AR_ForwardGrip:Exec
    MedPackPaddles:Exec
    SMG_P90:Exec
    AR_DrumGun:Exec
    Consumable_Small:Exec
    Consumable_Large:Exec
    Balloon:Exec
    MountedTurret:Exec
    CreativeTool:Exec
    ExplosiveBow:Exec
    AshtonIndigo:Exec
    AshtonChicago:Exec
    MeleeDualWield:Exec
    Unarmed:Exec
   )
   Set PickaxeEquipped(execute=[@n30.Melee, @n30.MeleeDualWield, @n30.MeleeOneHand], PickaxeEquipped=true)
   n31 = Event OnBeginCrouching()
   Event BlueprintInitializeAnimation()
   > Set onCrouchRotateAlpha(onCrouchRotateAlpha=0)
   > Set crouchEventBool(crouchEventBool=false)
   n32 = Event OnEndCrouching()
  ]}
]
variables [
 crouchEventBool {type=boolean}
 gearRotation {type=Rotator}
 IsCrouchMovingBool {type=boolean}
 isCrouchMovingFloat {type=number}
 L calf bend {type=number}
 L thigh twist {type=number}
 L_clavicle_follow {type=number}
 NeckBack {type=number}
 onCrouchRotateAlpha {type=number, default=1}
 PickaxeEquipped {type=boolean}
 R calf bend {type=number}
 R thigh twist {type=number}
 R_clavicle_follow {type=number}
 Rotation Out {type=Rotator}
 spineRotate left {type=number}
 spineRotate right {type=number}
]
